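; B_HeldObject_Generic is a generic held-object Blueprint that makes its static mesh match the object attached to its instigator.
; On BeginPlay it filters the instigator's attached actors by B_HeldObject_Parent and takes one from the result.
; It copies that actor's static mesh and world scale onto its own mesh component, so one class can mirror different held objects.
; Tick and the construction script hold no logic.
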(event EventBeginPlay
  (bind _staticmesh (Variables|Default|GetStaticMesh))
  (bind _filteredarray (Utilities|Array|FilterArray (Actor|GetAttachedActors (Game|GetInstigator self) true false) "/Game/Athena/Items/HeldObjects/B_HeldObject_Parent.B_HeldObject_Parent_C"))
  (bind _asb_held_object_parent (Utilities|Casting|CastToB_HeldObject_Parent (Utilities|Array|Get(acopy) _filteredarray)))
  (Components|StaticMesh|SetStaticMesh _staticmesh (Variables|/Script/Engine.StaticMeshComponent|GetStaticMesh (Variables|/Game/Athena/Items/HeldObjects/B_HeldObject_Parent.B_HeldObject_Parent_C|GetStaticMesh _asb_held_object_parent)))
  (Transformation|SetWorldScale3D _staticmesh (Transformation|GetWorldScale (Variables|/Game/Athena/Items/HeldObjects/B_HeldObject_Parent.B_HeldObject_Parent_C|GetStaticMesh _asb_held_object_parent))))

(event EventTick
  (params (DeltaSeconds :type Float%20%28single-precision%29)))

(fn ConstructionScript)
