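; Generic held-object weapon Blueprint that adds a "destroy" HUD key hint when the carried object is a Battle Lab device.
; On equip and attach, it waits on a short retriggerable delay for the held object to attach, then fetches it from the instigating player pawn.
; A valid object with the Athena.BattleLab gameplay tag makes the configured HUD key entry visible.
; Local unequip and the construction script contain no logic.
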(event FortWeapon|EventOnEquippedAndAttached
  ;[1 1 1] If this is a battle lab device, show the "destroy" hud key data.
  (Utilities|FlowControl|RetriggerableDelay 0.1) ; Wait for Held Object To Get Attached
  (bind _asfort_player_pawn (Utilities|Casting|CastToFortPlayerPawn (Game|GetInstigator)))
  (bind _returnvalue (Fortnite|HeldObjects|GetHeldObjectAttachedtoPlayer _asfort_player_pawn))
  (Utilities|IsValid _returnvalue
    (:Is%20Valid
      (bind _returnvalue_2 (GameplayTags|HasMatchingGameplayTag(Message) _returnvalue "(TagName=\"Athena.BattleLab\")"))
      (if _returnvalue_2
        (WeaponHUDData|SetHudKeyActionVisibility (Variables|Default|GetBattleLabHudKeyDataIndex) true)))))

(event FortWeapon|EventOnLocalUnEquip)

(fn ConstructionScript)
